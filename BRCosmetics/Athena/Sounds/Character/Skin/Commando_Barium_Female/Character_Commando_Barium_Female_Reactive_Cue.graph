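# Sound cue for the reactive sound of the female Commando Barium character skin.
# A Random node picks one of three equally weighted wave variations, and a Modulator varies its pitch and volume slightly.
# Foley attenuation spatializes it under the gadgets sound class, and a concurrency override allows one instance at a time, stopping the oldest.
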SoundCue /BRCosmetics/Athena/Sounds/Character/Skin/Commando_Barium_Female/Character_Commando_Barium_Female_Reactive_Cue
root=Modulator_0
nodes [
 {id=Modulator_0, children [Random_0], properties {pitchMax=1.1, volumeMax=1.1, volumeMin=1}}
 {id=Random_0, children [WavePlayer_3,WavePlayer_4,WavePlayer_5], properties {weights [1,1,1]}}
 {
  id=WavePlayer_3
  properties {soundWaveAssetPtr=/Game/Athena/Sounds/Character/Skin/Commando_Barium_Female/Character_Commando_BariumFemale_Reactive_01}
 }
 {
  id=WavePlayer_4
  properties {soundWaveAssetPtr=/Game/Athena/Sounds/Character/Skin/Commando_Barium_Female/Character_Commando_BariumFemale_Reactive_02}
 }
 {
  id=WavePlayer_5
  properties {soundWaveAssetPtr=/Game/Athena/Sounds/Character/Skin/Commando_Barium_Female/Character_Commando_BariumFemale_Reactive_03}
 }
]
properties {
 attenuationSettings=&Foley_Attenuation_Basic_Att
 bOverrideConcurrency=true
 maxDistance=1536
 soundClassObject=&Fort_Gadgets
 volumeMultiplier=0.7
 concurrencyOverrides {maxCount=1, resolutionRule=StopOldest}
}
